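# Translucent unlit Niagara sprite and mesh material, apparently for the SunShine pickaxe's idle effect.
# Two panning grayscale textures, on polar, particle-anchored screen or mesh UVs, are overlay-blended and eroded by a Fresnel/normal gradient or UV candle-shaped mask.
# Main and particle colors drive emissive; opacity adds backface-only masking plus depth, view-angle and near-camera fades.
# A world-up offset, masked by upward-facing normals, stretches vertices.
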Material /BRCosmetics/Weapons/FORT_Melee/Pickaxe_SunShine/FX/Test2/M_Sunshine_Idle
graph {
 properties {
  bUsedWithNiagaraMeshParticles=true
  bUsedWithNiagaraSprites=true
  blendMode=BLEND_Translucent
  pixelDepthOffsetMode=PDOM_Legacy
  shadingModel=MSM_Unlit
  usedShadingModels=Unlit
 }
 nodes [
  n1 = Constant3Vector() {constant=<rgba 0 0 1 0>}
  n2 = Saturate(None=DotProduct(A=VertexNormalWS(), B=@n1))
  n3 = ScalarParameter() {defaultValue=20, group=WPO, parameterName=Vertex Offset}
  LinearInterpolate(B=Multiply(A=Multiply(A=@n2, B=@n3), B=@n1), Alpha=ScalarParameter() {defaultValue=1, group=WPO, parameterName=UseVertexOffset}) {constB=0}
  n4 = ParticlePositionWS()
  n5 = StaticSwitchParameter(
   True=MaterialFunctionCall!(UV=TextureCoordinate(), Twist Center, Twist Edges) {materialFunction=/Game/Accessories/FORT_Backpacks/Backpack_F_MED_Celestial_Glow/FX/Materials/MF_PolarUV}.Polar UVs
   False=LinearInterpolate(
    A=MaterialFunctionCall!(Center Point World Position=@n4) {materialFunction=/Engine/Functions/Engine_MaterialFunctions02/UVs/WorldSpaceAlignedScreenCoordinates}.Offset UVs
    B=TextureCoordinate()
    Alpha=ScalarParameter() {group=Misc., parameterName=UseUVs?}
   )
  ) {group=Misc., parameterName=RadialCoordinates}
  MaterialFunctionCall(Float2=@n5) {materialFunction=&BreakOutFloat2Components}
  n6 = VectorParameter() {parameterName=Main Color, defaultValue=<rgba 2 0.2667 0 0.8>}
  n7 = ParticleColor()
  MaterialFunctionCall(Emissive=LinearInterpolate(A=Multiply(A=@n6.RGB, B=@n7.RGB), B=@n7.RGB, Alpha=ScalarParameter() {group="🎨Color", parameterName=UseParticleColor})) {materialFunction=&MF_FX_EmissiveAdjustments}
  n8 = OneMinus(None=Saturate(None=TwoSidedSign()))
  n9 = ScalarParameter() {defaultValue=1, group="🌌Misc.", parameterName=ShowBackfaceOnly}
  n10 = VertexNormalWS()
  n11 = CameraVectorWS()
  n12 = Power(Base=Abs(None=DotProduct(A=@n10, B=@n11)), Exp=ScalarParameter() {defaultValue=8, group="🔺Shape", parameterName="FresnelExponent (S)"})
  n13 = VertexNormalWS()
  n14 = ConstantBiasScale(None=ComponentMask(None=Transform(None=@n13) {transformSourceType=TRANSFORMSOURCE_World, transformType=TRANSFORM_View}) {g=true})
  n15 = ScalarParameter() {defaultValue=0.5, parameterName=Param_1}
  n16 = ComponentMask(None=TextureCoordinate()) {r=true}
  n17 = ComponentMask(None=TextureCoordinate()) {g=true}
  n18 = Multiply(
   A=Power(Base=Sine(None=@n16) {period=2}, Exp=ScalarParameter() {defaultValue=2, group="🔺Shape", parameterName=Candle Mask Side Power})
   B=Power(Base=OneMinus(None=@n17), Exp=ScalarParameter() {defaultValue=1.5, group="🔺Shape", parameterName=Candle Mask Power})
  )
  n19 = SmoothStep(Value=Power(Base=@n17)) {constMax=0.05}
  n20 = LinearInterpolate(
   A=Max(A=OneMinus(None=@n12), B=Multiply(A=@n14, B=@n15) {constB=0.5})
   B=OneMinus(None=Multiply(A=@n18, B=@n19))
   Alpha=ScalarParameter() {parameterName=useUVErosion?}
  )
  n21 = ScalarParameter() {defaultValue=0.4, group="🔺Shape", parameterName="Erosion Contrast (S)"}
  n22 = NamedRerouteUsage()
  n23 = Multiply(
   A=ScalarParameter() {defaultValue=50, group="🔺Shape", parameterName=Tex2_Size}
   B=ScalarParameter() {defaultValue=1, group="🔺Shape", parameterName=GlobalScale}
  )
  n24 = ScalarParameter() {group="🔺Shape", parameterName=Tex2 PanX}
  n25 = ScalarParameter() {defaultValue=0.5, group="🔺Shape", parameterName=Tex2 PanY}
  n26 = ScalarParameter() {parameterName=SwitchPanDirection}
  n27 = TextureSampleParameter2D(UVs=MaterialFunctionCall!(
    UV=Divide(A=@n22, B=@n23)
    Tile U=ScalarParameter() {defaultValue=1, group="🔺Shape", parameterName="Tex2_Tile U (S)"}
    Tile V=ScalarParameter() {defaultValue=1, group="🔺Shape", parameterName="Tex2_Tile V (S)"}
    Pan U=LinearInterpolate(A=@n24, B=@n25, Alpha=@n26)
    Pan V=LinearInterpolate(A=@n25, B=@n24, Alpha=@n26)
    Offset U
    Offset V
    Time=NamedRerouteUsage()
   ) {materialFunction=/Game/Effects/Fort_Effects/Materials/_Functions/MF_PanTileOffset}.UV's) {parameterName=MainTex2, samplerType=SAMPLERTYPE_LinearGrayscale, texture=&/sunboy}
  n28 = NamedRerouteUsage()
  n29 = Multiply(
   A=Multiply(
    A=ScalarParameter() {defaultValue=40, group="🔺Shape", parameterName=Tex1_Scale}
    B=ScalarParameter() {defaultValue=1, group="🔺Shape", parameterName=GlobalScale}
   )
   B=AppendVector(
    A=ScalarParameter() {defaultValue=0.8, group="🔺Shape", parameterName=Tex1_TileX}
    B=ScalarParameter() {defaultValue=1.3, group="🔺Shape", parameterName=Tex1_TileY}
   )
  )
  n30 = ScalarParameter() {group="🔺Shape", parameterName=Tex1_PanX}
  n31 = ScalarParameter() {defaultValue=1, group="🔺Shape", parameterName=Tex1_PanY}
  n32 = AppendVector(A=@n30, B=@n31)
  n33 = MaterialFunctionCall(XY=MaterialFunctionCall(X=@n30, Y=@n31) {materialFunction=&MakeFloat2}.Result) {materialFunction=&Swizzle}
  n34 = ScalarParameter() {parameterName=SwitchPanDirection}
  n35 = TextureSampleParameter2D(UVs=Panner(Coordinate=Divide(A=@n28, B=@n29), Time=NamedRerouteUsage(), Speed=LinearInterpolate(A=@n32, B=@n33.YX, Alpha=@n34)) {speedY=0.8}) {
   parameterName=MainTex
   samplerType=SAMPLERTYPE_LinearGrayscale
   texture=&/T_FX_AlmondBoy
  }
  n36 = MaterialFunctionCall(
   Base=LinearInterpolate(A=@n35.G, B=OneMinus(None=@n35.G), Alpha=ScalarParameter() {defaultValue=1, group="🔺Shape", parameterName=InvertMainTexture})
   Blend=@n27.G
  ) {materialFunction=&Blend_Overlay}
  n37 = Multiply(
   A=@n6.A
   B=Multiply(
    A=LinearInterpolate(B=@n8, Alpha=@n9) {constA=1}
    B=MaterialFunctionCall!(Gradient=@n36.Result, Erosion Amount=@n20, Softness=@n21) {materialFunction=/BRCosmetics/Athena/Effects/Particles/Skydiving/NS_Contrail_DarkAndLight/Functions/MF_Erosion}.Increased Range
   )
  )
  n38 = DepthFade(Opacity=@n37, FadeDistance=ScalarParameter() {defaultValue=5, group=Misc., parameterName=DepthFadeDistance})
  n39 = ShadingPathSwitch(Default=@n38, Deferred=@n38, Forward=@n37, Mobile=@n37)
  n40 = MaterialFunctionCall("FadeSensitivity (optional)"=ScalarParameter() {group=Misc., parameterName="FadeSensitivity (optional) (S)"}) {materialFunction=&FadeBasedOnViewAngle}
  n41 = ScalarParameter() {defaultValue=100, group=Misc., parameterName="Distance (S)"}
  Multiply(
   A=@n7.A
   B=Multiply(
    A=Multiply(A=@n39, B=@n40.Result)
    B=MaterialFunctionCall(Distance=@n41) {materialFunction=&Near_Camera_Fade}.Result
   )
  )
  n42 = MaterialFunctionCall(Float2=@n5) {materialFunction=&BreakOutFloat2Components}
  n43 = Multiply(A=@n42.G) {constB=-1}
  NamedRerouteDeclaration(None=LinearInterpolate(
    A=@n5
    B=MaterialFunctionCall(X=@n42.R, Y=@n43) {materialFunction=&MakeFloat2}.Result
    Alpha=ScalarParameter() {parameterName=SwizzleUVs?}
   )) {name=UVs, nodeColor=<rgba 0.9688 1 0 1>}
  NamedRerouteDeclaration(None=Multiply(A=Time(), B=ScalarParameter() {defaultValue=0.5, parameterName=TimeScale}) {constB=0.2}) {name=GlobalTimeScale, nodeColor=<rgba 0 0.9766 1 1>}
 ]
}
